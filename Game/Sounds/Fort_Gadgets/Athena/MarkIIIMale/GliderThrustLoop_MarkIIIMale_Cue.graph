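# Looping glider thrust sound cue for the MarkIIIMale glider in Athena.
# A mixer boosts and blends two looping flight layers, a base and a Hi variant, each through its own continuous modulator.
# The AnyDirection parameter raises both layers' pitch and drives volume, varying the base layer slightly and the Hi layer more widely.
# Playback uses the player glider thrust attenuation and the gliders sound class.
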SoundCue /Game/Sounds/Fort_Gadgets/Athena/MarkIIIMale/GliderThrustLoop_MarkIIIMale_Cue
root=Mixer_0
nodes [
 {id=Mixer_0, children [ModulatorContinuous_0, ModulatorContinuous_1], properties {inputVolume [2,2]}}
 {
  id=ModulatorContinuous_0
  children [WavePlayer_0]
  properties {
   pitchModulationParams {default=0, maxOutput=1.2, minOutput=1, parameterName=AnyDirection}
   volumeModulationParams {default=0, minOutput=0.9, parameterName=AnyDirection}
  }
 }
 {
  id=ModulatorContinuous_1
  children [WavePlayer_2]
  properties {
   pitchModulationParams {default=0, maxOutput=1.2, minOutput=1, parameterName=AnyDirection}
   volumeModulationParams {default=0, maxOutput=1.2, parameterName=AnyDirection}
  }
 }
 {id=WavePlayer_0, properties {bLooping=true, soundWaveAssetPtr=&/Glider_MarkIIIMale_Flight_LP}}
 {id=WavePlayer_2, properties {bLooping=true, soundWaveAssetPtr=&/Glider_MarkIIIMale_Flight_LP_Hi}}
]
properties {
 attenuationSettings=&Fort_Player_Glider_Thrust_Athena_Att
 maxDistance=12500
 soundClassObject=&Fort_Gliders_Land
 volumeMultiplier=1
 concurrencySet [null]
}
